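# Physics asset for the Takoyaki backpack cosmetic, with simple collision bodies for the pack and its dynamic bones.
# A box covers the pack and a large sphere covers dyn_pack_base, ringed by small spheres on the up, down, left, right and front bones.
# The pack and those five small spheres each constrain to the base with swing and twist locked.
# A capsule fork hangs off dyn_pack_up on its own constraint with small swing and twist limit angles.
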PhysicsAsset /BRCosmetics/Accessories/FORT_Backpacks/Backpack_M_MED_Takoyaki/Meshes/M_MED_Takoyaki_Pack_Physics
bodies {
 dyn_fork [{shapeType=Capsule, length=8.75, radius=2.25, shapeName=dyn_fork_capsule, center=<4.8166 0 0>, rotation=<rot -90 179.9602 -179.9607>}]
 dyn_pack_base [{shapeType=Sphere, radius=10.5, shapeName=dyn_pack_base_sphere}]
 dyn_pack_down [{shapeType=Sphere, radius=2.5, shapeName=dyn_pack_down_sphere}]
 dyn_pack_front [{shapeType=Sphere, radius=2.5, shapeName=dyn_pack_front_sphere}]
 dyn_pack_left [{shapeType=Sphere, radius=2.5, shapeName=dyn_pack_left_sphere}]
 dyn_pack_right [{shapeType=Sphere, radius=2.5, shapeName=dyn_pack_right_sphere}]
 dyn_pack_up [{shapeType=Sphere, radius=2.5, shapeName=dyn_pack_up_sphere}]
 pack [{shapeType=Box, extentX=15.5, extentY=2, extentZ=23, shapeName=pack_box}]
}
constraints [
 {
  bone1Name=dyn_pack_right
  bone2Name=dyn_pack_base
  swing1LimitDegrees=45
  swing1Motion=Locked
  swing2LimitDegrees=45
  swing2Motion=Locked
  twistLimitDegrees=45
  twistMotion=Locked
 }
 {
  bone1Name=dyn_pack_left
  bone2Name=dyn_pack_base
  swing1LimitDegrees=45
  swing1Motion=Locked
  swing2LimitDegrees=45
  swing2Motion=Locked
  twistLimitDegrees=45
  twistMotion=Locked
 }
 {
  bone1Name=dyn_pack_down
  bone2Name=dyn_pack_base
  swing1LimitDegrees=45
  swing1Motion=Locked
  swing2LimitDegrees=45
  swing2Motion=Locked
  twistLimitDegrees=45
  twistMotion=Locked
 }
 {
  bone1Name=dyn_pack_up
  bone2Name=dyn_pack_base
  swing1LimitDegrees=45
  swing1Motion=Locked
  swing2LimitDegrees=45
  swing2Motion=Locked
  twistLimitDegrees=45
  twistMotion=Locked
 }
 {
  bone1Name=dyn_pack_front
  bone2Name=dyn_pack_base
  swing1LimitDegrees=45
  swing1Motion=Locked
  swing2LimitDegrees=45
  swing2Motion=Locked
  twistLimitDegrees=45
  twistMotion=Locked
 }
 {bone1Name=dyn_fork, bone2Name=dyn_pack_up, swing1LimitDegrees=10, swing2LimitDegrees=10, twistLimitDegrees=5}
 {
  bone1Name=pack
  bone2Name=dyn_pack_base
  swing1LimitDegrees=45
  swing1Motion=Locked
  swing2LimitDegrees=45
  swing2Motion=Locked
  twistLimitDegrees=45
  twistMotion=Locked
 }
]
