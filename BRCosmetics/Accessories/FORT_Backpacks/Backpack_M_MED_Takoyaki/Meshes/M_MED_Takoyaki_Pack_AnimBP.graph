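# Animation Blueprint for the Takoyaki backpack cosmetic that adds physics-driven secondary motion to the pack skeleton.
# The AnimGraph runs the input pose through a RigidBody node with an override physics asset and gravity override.
# Each update, the EventGraph lowers the simulation's blend alpha while skydiving or falling and restores it otherwise.
# The node interpolates alpha changes to smooth the transition.
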AnimBlueprint /BRCosmetics/Accessories/FORT_Backpacks/Backpack_M_MED_Takoyaki/Meshes/M_MED_Takoyaki_Pack_AnimBP
target_skeleton=&/M_MED_Takoyaki_Pack_Skeleton
graphs [
 AnimGraph {nodes [
   n1 = RigidBody(ComponentPose=Local To Component(LocalPose=Input Pose()), Alpha=Get Dyn_Rigid_Body_Alfa() -> (Dyn_Rigid_Body_Alfa:double)) {
    bOverrideWorldGravity=true
    overridePhysicsAsset=&/M_MED_Takoyaki_Pack_Physics
    alphaBoolBlend {blendOption=Linear}
    alphaScaleBiasClamp {bInterpResult=true, interpSpeedDecreasing=20, interpSpeedIncreasing=20}
    componentLinearAccScale=<0.75 0.75 0.5>
    componentLinearVelScale=<1 1 1>
    simSpaceSettings {maxAngularAcceleration=1, maxAngularVelocity=1, maxLinearAcceleration=1, maxLinearVelocity=1}
   }
   Output Pose(Result=Component To Local(ComponentPose=@n1)) {layerGroup=None}
  ]}
 EventGraph {nodes [
   Event BlueprintUpdateAnimation()
   > n1 = Branch(Condition=OR Boolean(A=Get bIsSkyDiving(), B=Get bIsFalling()))
   > Set Dyn_Rigid_Body_Alfa(Dyn_Rigid_Body_Alfa:double=0.5) -> (Output_Get:double)
   Set Dyn_Rigid_Body_Alfa(execute=@n1.else, Dyn_Rigid_Body_Alfa=0.9)
   TryGetPawnOwner()
  ]}
]
variables [Dyn_Rigid_Body_Alfa {type=number, default=0.9}]
